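# Surface material for the Hypertech cosmetic vehicle wheel, with a pulsing tinted glow on the rim.
# MF_WheelParent_Simple supplies base attributes, a rim mask and vertex color.
# Emissive blends base color, by inverted red vertex color, with alpha-masked base color pushed toward a scalable olive tint by an unclamped time sine.
# The result runs through the shared vehicle damage function, then the de-rez transition.
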Material /VehicleCosmetics/Wheels/Hypertech/Materials/M_Wheel_Hypertech
graph {
 properties {
  bAutomaticallySetUsageInEditor=false
  bUseMaterialAttributes=true
  bUsedWithMeshDeformer=true
  bUsedWithSkeletalMesh=true
  pixelDepthOffsetMode=PDOM_Legacy
  usedShadingModels=Default Lit
 }
 nodes [
  n1 = MaterialFunctionCall() {materialFunction=&MF_WheelParent_Simple}
  n2 = GetMaterialAttributes(None=@n1.MaterialAttributes) {attributeGetTypes [#1]}
  n3 = ComponentMask(None=@n1.RimRGBAMask) {a=true}
  n4 = Multiply(A=@n2.Base Color, B=@n3)
  n5 = Multiply(
   A=ScalarParameter() {defaultValue=1, group=5_Special Effects, parameterName=Emissive_Value}
   B=Constant3Vector() {constant=<rgba 0.306 0.349 0.161 0>}
  )
  n6 = Sine(None=Multiply(A=Time()) {constB=0.4})
  n7 = LinearInterpolate(
   A=@n2.Base Color
   B=LinearInterpolate(A=@n4, B=Add(A=@n4, B=@n5) {constB=0.3}, Alpha=Multiply(A=@n3, B=@n6))
   Alpha=OneMinus(None=@n1.VertexColorR)
  )
  MaterialFunctionCall(In=MaterialFunctionCall(Material=SetMaterialAttributes(MaterialAttributes=@n2.MaterialAttributes, Emissive Color=@n7) {attributeSetTypes [#2]}) {materialFunction=&MF_BaseVehicle_Damage}.Result) {materialFunction=&MF_VehicleDeRez_Transition}
 ]
}
