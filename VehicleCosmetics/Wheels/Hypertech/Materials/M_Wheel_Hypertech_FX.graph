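# Translucent material for the Hypertech wheel's spinning disc effect, with DiscColor and Opacity parameters.
# Emissive blends DiscColor with a darker offset via a gradient texture, restoring full color under the rotating rim mask.
# Opacity combines that mask with gradients from counter-rotating UVs, then multiplies by fine concentric rings from fractional UV radial distance.
# A blurred-circle texture softens the shape.
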Material /VehicleCosmetics/Wheels/Hypertech/Materials/M_Wheel_Hypertech_FX
graph {
 properties {
  bAutomaticallySetUsageInEditor=false
  bUseMaterialAttributes=true
  bUsedWithMeshDeformer=true
  bUsedWithSkeletalMesh=true
  blendMode=BLEND_Translucent
  pixelDepthOffsetMode=PDOM_Legacy
  usedShadingModels=Default Lit
 }
 nodes [
  n1 = VectorParameter() {group=0_Special Effects, parameterName=DiscColor, sortPriority=1, defaultValue=<rgba 1 1 0.2016 1>}
  n2 = Subtract(A=@n1.RGB, B=Constant3Vector() {constant=<rgba 0 0.75 0.15 0>})
  n3 = TextureSample() {texture=/VehicleCosmetics/Wheels/_SharedTextures/FX/Darkness_Gradient}
  n4 = TextureCoordinate()
  n5 = Rotator(Coordinate=@n4) {speed=50}
  n6 = TextureSample(UVs=@n5) {samplerType=SAMPLERTYPE_Masks, texture=/VehicleCosmetics/Wheels/Hypertech/Textures/T_Rim_Hypertech_Mask}
  n7 = Constant() {r=0.3}
  n8 = ComponentMask(None=@n5) {r=true}
  n9 = Subtract(A=@n8) {constB=0.5}
  n10 = Subtract(B=@n8) {constA=0.5, constB=0.8}
  n11 = ComponentMask(None=Rotator(Coordinate=@n4) {speed=-50}) {r=true}
  n12 = Subtract(A=@n11) {constB=0.5}
  n13 = Subtract(B=@n11) {constA=0.5, constB=0.8}
  n14 = AppendVector(A=Abs(None=AppendVector(A=@n9, B=@n10)), B=Abs(None=AppendVector(A=@n12, B=@n13)))
  n15 = TextureCoordinate()
  n16 = Constant2Vector() {g=0.5, r=0.5}
  n17 = Multiply(A=Length(None=Subtract(A=@n15, B=@n16)) {is=class MaterialExpressionLength}, B=Constant() {r=80}) {constB=5}
  n18 = Multiply(A=Add(A=Multiply(A=@n7, B=@n6.R), B=Multiply(A=@n14) {constB=0.6}), B=ConstantBiasScale(None=Frac(None=@n17)) {scale=0.4})
  n19 = ScalarParameter() {defaultValue=1, group=0_Special Effects, parameterName=Opacity, sortPriority=3}
  MakeMaterialAttributes!(
   BaseColor
   Metallic
   Specular
   Roughness
   Anisotropy
   EmissiveColor=LinearInterpolate(A=LinearInterpolate(A=@n1.RGB, B=@n2, Alpha=@n3.R), B=@n1.RGB, Alpha=Multiply(A=@n6.R) {constB=0.7})
   Opacity=Multiply(A=Multiply(A=@n18, B=@n19), B=TextureSample() {texture=/VehicleCosmetics/Wheels/_SharedTextures/FX/BlurredCircle}.R)
   OpacityMask
   Normal
   Tangent
   WorldPositionOffset
   SubsurfaceColor
   ClearCoat
   ClearCoatRoughness
   AmbientOcclusion
   Refraction
   CustomizedUVs_0
   CustomizedUVs_1
   CustomizedUVs_2
   CustomizedUVs_3
   CustomizedUVs_4
   CustomizedUVs_5
   CustomizedUVs_6
   CustomizedUVs_7
   PixelDepthOffset
   ShadingModel
   Displacement
  )
 ]
}
